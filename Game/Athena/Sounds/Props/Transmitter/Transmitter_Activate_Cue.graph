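# Sound cue for activating a transmitter prop, mixing a placement hit, a reward stinger and ongoing machine layers within a short falloff range.
# The placement hit uses its own non-spatialized attenuation, and the reward stinger follows after a fixed delay.
# Beep and two running loops fade in through envelopers, while randomly chosen handshake sounds repeat endlessly after a random initial delay.
# Cue-level overrides add listener focus, distance low-pass, occlusion and stop-quietest concurrency.
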SoundCue /Game/Athena/Sounds/Props/Transmitter/Transmitter_Activate_Cue
root=Mixer_0
nodes [
 {
  id=Attenuation_0
  children [Modulator_0]
  properties {
   bOverrideAttenuation=true
   attenuationOverrides {bSpatialize=false, distanceAlgorithm=Logarithmic, falloffDistance=1100, attenuationShapeExtents {x=0}}
  }
 }
 {id=Delay_0, children [WavePlayer_4], properties {delayMax=0.4, delayMin=0.4}}
 {id=Delay_1, children [Looping_0], properties {delayMax=1}}
 {
  id=Enveloper_0
  children [WavePlayer_1]
  properties {pitchCurve {editorCurveData {keys [{time=0}]}}, volumeCurve {editorCurveData {keys [<at 0.0233 0.0116>,<at 1.0037 0.9882>]}}}
 }
 {
  id=Enveloper_1
  children [WavePlayer_0]
  properties {pitchCurve {editorCurveData {keys [<at 0.2392 1.2207>]}}, volumeCurve {editorCurveData {keys [<at 0.0233 0.0116>,<at 1.5065 0.9996>]}}}
 }
 {
  id=Enveloper_2
  children [WavePlayer_3]
  properties {pitchCurve {editorCurveData {keys [<at 0.2394 1.2173>]}}, volumeCurve {editorCurveData {keys [<at 0.0233 0.0116>,<at 0.7456 0.9951>]}}}
 }
 {id=Looping_0, children [Random_0]}
 {
  id=Mixer_0
  children [Attenuation_0, Enveloper_0, Enveloper_2, Enveloper_1, Delay_0, Delay_1]
  properties {inputVolume [1.2,0.2,1.2,0.8,0.5,0.3]}
 }
 {id=Modulator_0, children [WavePlayer_2], properties {pitchMax=1, pitchMin=1, volumeMax=1, volumeMin=1}}
 {
  id=Random_0
  children [WavePlayer_5, WavePlayer_6, WavePlayer_7, WavePlayer_8, WavePlayer_9, WavePlayer_10, WavePlayer_11, WavePlayer_12]
  properties {weights [1,1,1,1,1,1,1,1]}
 }
 {
  id=WavePlayer_0
  properties {bLooping=true, soundWaveAssetPtr=/Game/Athena/Sounds/Props/MountainBaseComputer/ComputerTransimtter_Running_Loop_Low_01}
 }
 {
  id=WavePlayer_1
  properties {bLooping=true, soundWaveAssetPtr=/Game/Athena/Sounds/Props/MountainBaseComputer/ComputerTransimtter_Beep_Loop_01}
 }
 {id=WavePlayer_10, properties {soundWaveAssetPtr=/Game/Athena/Sounds/Props/MountainBaseComputer/ComputerTransmitter_Handshake_05}}
 {id=WavePlayer_11, properties {soundWaveAssetPtr=/Game/Athena/Sounds/Props/MountainBaseComputer/ComputerTransmitter_Handshake_06}}
 {id=WavePlayer_12, properties {soundWaveAssetPtr=/Game/Athena/Sounds/Props/MountainBaseComputer/ComputerTransmitter_Handshake_07}}
 {id=WavePlayer_2, properties {soundWaveAssetPtr=/Game/Athena/Sounds/Props/MountainBaseComputer/ComputerTransimtter_Placed_01}}
 {
  id=WavePlayer_3
  properties {bLooping=true, soundWaveAssetPtr=/Game/Athena/Sounds/Props/MountainBaseComputer/ComputerTransimtter_Running_Loop_01}
 }
 {id=WavePlayer_4, properties {soundWaveAssetPtr=/Game/Sounds/Fort_UI_Sounds/Rewards/Fort_RewardScreen_TeamScore_ChestLootUpgrade}}
 {id=WavePlayer_5, properties {soundWaveAssetPtr=/Game/Athena/Sounds/Props/MountainBaseComputer/ComputerTransmitter_Handshake_08}}
 {id=WavePlayer_6, properties {soundWaveAssetPtr=/Game/Athena/Sounds/Props/MountainBaseComputer/ComputerTransmitter_Handshake_01}}
 {id=WavePlayer_7, properties {soundWaveAssetPtr=/Game/Athena/Sounds/Props/MountainBaseComputer/ComputerTransmitter_Handshake_02}}
 {id=WavePlayer_8, properties {soundWaveAssetPtr=/Game/Athena/Sounds/Props/MountainBaseComputer/ComputerTransmitter_Handshake_03}}
 {id=WavePlayer_9, properties {soundWaveAssetPtr=/Game/Athena/Sounds/Props/MountainBaseComputer/ComputerTransmitter_Handshake_04}}
]
properties {
 bOverrideAttenuation=true
 bOverrideConcurrency=true
 maxDistance=1100
 priority=50
 soundClassObject=&Fort_Loot_Reward_PSM
 volumeMultiplier=1
 attenuationOverrides {
  bEnableFocusInterpolation=true
  bEnableListenerFocus=true
  distanceAlgorithm=Logarithmic
  falloffDistance=1100
  focusAttackInterpSpeed=20
  focusAzimuth=20
  focusReleaseInterpSpeed=10
  focusVolumeAttenuation=1.75
  lPFFrequencyAtMax=10000
  lPFRadiusMax=1000
  lPFRadiusMin=500
  nonFocusAzimuth=40
  nonFocusVolumeAttenuation=0.8
  occlusionInterpolationTime=0.3
  occlusionLowPassFilterFrequency=3000
  occlusionVolumeAttenuation=0.75
  stereoSpread=10000
  attenuationShapeExtents {x=0}
  customAttenuationCurve {editorCurveData {keys [<at 0 1 RCIM_Cubic>, {arriveTangent=-1, interpMode=RCIM_Cubic, leaveTangent=-1, time=0.5, value=0.1}, <at 1 0 RCIM_Cubic>]}}
 }
 concurrencyOverrides {resolutionRule=StopQuietest}
}
